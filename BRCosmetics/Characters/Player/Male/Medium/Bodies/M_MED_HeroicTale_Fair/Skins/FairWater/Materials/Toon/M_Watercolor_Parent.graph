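# Parent watercolor toon material for the FairWater character skin, using material attributes and the Cloth shading model on skeletal meshes.
# The shared 2023 anime shader's Pre and Post functions wrap a watercolor calculations function that supplies the shading gradients.
# The Mask texture's blue channel marks skin, feeding the calculations and tinting shaded color toward a separate skin shade.
# A static switch optionally applies WindEffect to the result, driven by selectable SkinFX_Mask channels.
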Material /BRCosmetics/Characters/Player/Male/Medium/Bodies/M_MED_HeroicTale_Fair/Skins/FairWater/Materials/Toon/M_Watercolor_Parent
graph {
 properties {
  bAutomaticallySetUsageInEditor=false
  bUseMaterialAttributes=true
  bUsedWithMorphTargets=true
  bUsedWithSkeletalMesh=true
  pixelDepthOffsetMode=PDOM_Legacy
  shadingModel=MSM_Cloth
  usedShadingModels=Cloth
 }
 nodes [
  n1 = TextureSampleParameter2D() {group=Skin FX, parameterName=SkinFX_Mask, samplerType=SAMPLERTYPE_Masks, texture=&T_BlackMask}
  n2 = AppendVector(A=@n1.RGB, B=@n1.RGB)
  n3 = TextureSampleParameter2D() {
   group=00_Texture Inputs
   parameterName=Mask
   samplerSource=SSM_Wrap_WorldGroupSettings
   samplerType=SAMPLERTYPE_Masks
   sortPriority=3
   texture=&T_BlackMask
  }
  n4 = NamedRerouteDeclaration(None=@n3.B) {name=SkinMask, nodeColor=<rgba 1 0 0.4219 1>}
  n5 = MaterialFunctionCall!(CustomUV) {materialFunction=/Game/Global/Materials/Character_Parent_2023/Anime_Parent_2023/MaterialFunctions/MFC_AnimeShader_2023_Pre}
  n6 = MaterialFunctionCall!(OverrideNormal_WS, SkinMask=@n4, CustomLightVector, STT=@n5.STT, Normal_Map=@n5.STT, DFL=@n5.STT, FX=@n5.STT) {materialFunction=&/MFC_WatercolorShader_Calculations}
  n7 = VectorParameter() {group=01_Anime Maps, parameterName=Color_Shaded_Skin, sortPriority=11, defaultValue=<rgba 0.3021 0.0488 0.0488 0>}
  n8 = MaterialFunctionCall!(
   Shading_Gradient=@n6.Shading_Gradient
   Threshold_Gradient=@n6.Shading_Gradient
   Specular_Gradient=@n6.Shading_Gradient
   Sheen_Mask=@n6.Shading_Gradient
   Thickness=@n6.Shading_Gradient
   DFL_Mask=@n6.Shading_Gradient
   FX=@n6.Shading_Gradient
   Metallic_Gradient=@n6.Shading_Gradient
   RimLight_Gradient=@n6.Shading_Gradient
   NormalBuffer=@n6.Shading_Gradient
   --------
   CINE_LightSources=@n6.Shading_Gradient
   CINE_LightColor=@n6.Shading_Gradient
   ---------
   Color_Lit=@n5.Color_Lit
   Color_Shaded=LinearInterpolate(A=@n5.Color_Shaded, B=@n7.RGB, Alpha=@n4)
   SRM=@n5.Color_Lit
   Emissive=@n5.Color_Lit
   DFL_Color=@n5.Color_Lit
   Specular_Color=@n5.Color_Lit
   RimLight_Color=@n5.Color_Lit
  ) {materialFunction=/Game/Global/Materials/Character_Parent_2023/Anime_Parent_2023/MaterialFunctions/MFC_AnimeShader_2023_Post}
  StaticSwitchParameter(
   True=MaterialFunctionCall!(
    MASK=StaticComponentMaskParameter(None=@n2) {defaultR=true, group=Skin FX, parameterName=WindEffect_Channel}
    Wind Intensity Mask=StaticComponentMaskParameter(None=@n2) {defaultR=true, group=Skin FX, parameterName=WindEffect Intensity_Channel}
    Material=@n8.Result
   ) {materialFunction=/Game/Weapons/WeaponSkins/MaterialFunction/WindEffect}.Result
   False=@n8.Result
  ) {group=Skin FX, parameterName=UseWindEffect}
 ]
}
